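# MetaSound source preset for the drop sound of the Grape item in SalineCabinet.
# It inherits its graph from the MSS_Item_Drop parent and only supplies inputs: three drop wave variations, master volume and a small random pitch spread.
# It outputs mono one-shot audio using weapon-drop attenuation, the loot-collect sound class and item concurrency.
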MetaSoundSource /SalineCabinet/Assets/Audio/Grape/MSS_SalineCabinet_Grape_Drop_Preset
preset_parent=/FortMetasound/Items/MSS_Item_Drop
inputs [
 MasterVolume {type=Float}
 Random.PitchShift.Base {type=Float}
 Random.PitchShift.Spread {type=Float}
 Sounds {type=WaveAsset:Array}
 UE.Source.OnPlay {type=Trigger}
]
interfaces [UE.OutputFormat.Mono:1.0, UE.Source.OneShot:1.0, UE.Source:1.1]
outputs [
 "UE.OutputFormat.Mono.Audio:0" {type=Audio}
 UE.Source.OneShot.OnFinished {type=Trigger}
]
pages [Default {input_defaults {
   MasterVolume=0.6
   Random.PitchShift.Base=0
   Random.PitchShift.Spread=0.5
   UE.Source.OnPlay=null
   Sounds [
    &/SalineCabinet_Grape_Drop_01
    &/SalineCabinet_Grape_Drop_02
    &/SalineCabinet_Grape_Drop_03
   ]
  }}]
properties {
 assetImportData=&self:AssetImportData
 attenuationSettings=&Fort_Weapon_Drop_Att
 maxDistance=3000
 soundClassObject=&Fort_Loot_Collect
 concurrencySet [&CS_Fort_Item]
}
